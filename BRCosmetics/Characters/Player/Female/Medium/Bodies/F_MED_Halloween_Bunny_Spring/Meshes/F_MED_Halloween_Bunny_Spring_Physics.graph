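# Physics asset for the F_MED_Halloween_Bunny_Spring female body, setting up secondary motion for a bow with ribbons on the right thigh.
# Small capsules cover the bow hub, its two loops and the front and back ribbon chains, with a large capsule on thigh_r.
# The hub rotates freely on the thigh and anchors both loops and chains through tight angular limits.
# An extra constraint cross-links the second back and front ribbon segments.
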PhysicsAsset /BRCosmetics/Characters/Player/Female/Medium/Bodies/F_MED_Halloween_Bunny_Spring/Meshes/F_MED_Halloween_Bunny_Spring_Physics
bodies {
 dyn_bow_l [{shapeType=Capsule, length=1, radius=1.5, center=<0.7 0 0>, rotation=<rot 90 0 0>}]
 dyn_bow_r [{shapeType=Capsule, length=1, radius=1.5, center=<1 0 0>, rotation=<rot 90 0 0>}]
 dyn_ribbon_bck_01 [{shapeType=Capsule, length=1, radius=1, center=<0.3 0 0>, rotation=<rot 90 0 0>}]
 dyn_ribbon_bck_02 [{shapeType=Capsule, length=1, radius=1, center=<0.5 0 0>, rotation=<rot 90 0 0>}]
 dyn_ribbon_frt_01 [{shapeType=Capsule, length=1, radius=1, center=<0.15 0 0>, rotation=<rot 90 0 0>}]
 dyn_ribbon_frt_02 [{shapeType=Capsule, length=1, radius=1, center=<0.3 0 0>, rotation=<rot 90 0 0>}]
 dyn_ribbon_frt_03 [{shapeType=Capsule, length=1, radius=1, center=<0.5 0 0>, rotation=<rot 90 0 0>}]
 dyn_thigh_bow [{shapeType=Capsule, length=0.505, radius=0.505, rotation=<rot 0 -90 0>}]
 thigh_r [{shapeType=Capsule, length=16, radius=8.25, center=<19.5203 -0.7645 0>, rotation=<rot 90 0 0>}]
}
constraints [
 {bone1Name=dyn_bow_l, bone2Name=dyn_thigh_bow, swing1LimitDegrees=25, swing2LimitDegrees=25, twistLimitDegrees=5}
 {bone1Name=dyn_bow_r, bone2Name=dyn_thigh_bow, swing1LimitDegrees=25, swing2LimitDegrees=25, twistLimitDegrees=5}
 {bone1Name=dyn_ribbon_frt_01, bone2Name=dyn_thigh_bow, swing1LimitDegrees=15, swing2LimitDegrees=10, twistLimitDegrees=2}
 {bone1Name=dyn_ribbon_frt_02, bone2Name=dyn_ribbon_frt_01, swing1LimitDegrees=10, swing2LimitDegrees=22, twistLimitDegrees=2}
 {bone1Name=dyn_ribbon_frt_03, bone2Name=dyn_ribbon_frt_02, swing1LimitDegrees=30, swing2LimitDegrees=30, twistLimitDegrees=2}
 {
  bone1Name=dyn_thigh_bow
  bone2Name=thigh_r
  swing1LimitDegrees=45
  swing1Motion=Free
  swing2LimitDegrees=45
  swing2Motion=Free
  twistLimitDegrees=45
  twistMotion=Free
 }
 {bone1Name=dyn_ribbon_bck_01, bone2Name=dyn_thigh_bow, swing1LimitDegrees=15.2857, swing2LimitDegrees=15, twistLimitDegrees=3}
 {bone1Name=dyn_ribbon_bck_02, bone2Name=dyn_ribbon_bck_01, swing1LimitDegrees=16.7143, swing2LimitDegrees=15, twistLimitDegrees=3}
 {bone1Name=dyn_ribbon_bck_02, bone2Name=dyn_ribbon_frt_02, swing1LimitDegrees=20, swing2LimitDegrees=20, twistLimitDegrees=20}
]
