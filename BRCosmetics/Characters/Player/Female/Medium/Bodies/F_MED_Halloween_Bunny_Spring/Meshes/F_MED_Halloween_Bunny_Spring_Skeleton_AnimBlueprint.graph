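# Animation blueprint for the Halloween Bunny Spring female character skeleton, adding secondary motion to the outfit.
# It copies the attached parent's mesh pose, then constrains the thigh bow between right thigh and twist bones.
# Spring-driven AnimDynamics nodes move the arm bandana and boot strings, followed by a RigidBody simulation using its own physics asset.
# The event graph sets crouch_alpha from a crouching flag, but the anim graph leaves it unconnected.
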AnimBlueprint /BRCosmetics/Characters/Player/Female/Medium/Bodies/F_MED_Halloween_Bunny_Spring/Meshes/F_MED_Halloween_Bunny_Spring_Skeleton_AnimBlueprint
target_skeleton=&/F_MED_Halloween_Bunny_Spring_Skeleton
graphs [
 AnimGraph {nodes [
   n1 = "Constraint - dyn_thigh_bow (PP)"!(
    ComponentPose=Local To Component(LocalPose=Copy Pose From Mesh() {bUseAttachedParent=true})
    Alpha=1
    ConstraintWeights_0=0.4137
    ConstraintWeights_1:float=0.5863
   ) -> (Pose) {
    alphaBoolBlend {blendOption=Linear}
    boneToModify {boneName=dyn_thigh_bow}
    constraintSetup [{targetBone {boneName=thigh_r}}, {targetBone {boneName=thigh_twist_01_r}}]
    constraintWeights [1,1]
   }
   n2 = Anim Dynamics - Bone: dyn_arm_r_bandana(ComponentPose=@n1, Alpha=0.75, ExternalForce="0,0,0") {
    alpha=0.9
    angularDampingOverride=0.985
    angularSpringConstant=300
    bAngularSpring=true
    bGravityOverrideInSimSpace=true
    bOverrideAngularDamping=true
    gravityScale=0.3
    numSolverIterationsPostUpdate=4
    numSolverIterationsPreUpdate=16
    alphaBoolBlend {blendOption=Linear}
    boundBone {boneName=dyn_arm_r_bandana}
    physicsBodyDefinitions [{
      sphereCollisionRadius=0
      boundBone {boneName=dyn_arm_r_bandana}
      boxExtents=<20 8 8>
      constraintSetup {angularTargetAxis=AxisZ, twistAxis=AxisY, angularLimitsMax=<15 0 15>, angularLimitsMin=<-15 0 -15>, angularTarget=<0 0 1>}
      localJointOffset=<0 0 -3>
     }]
   }
   n3 = Anim Dynamics - Bone: dyn_boot_strings_01(ComponentPose=@n2, Alpha=0.55, ExternalForce="0,0,0") {
    bGravityOverrideInSimSpace=true
    bLinearSpring=true
    bOverrideLinearDamping=true
    gravityScale=0.33
    linearDampingOverride=0.995
    linearSpringConstant=150
    simulationSpace=RootRelative
    alphaBoolBlend {blendOption=Linear}
    boundBone {boneName=dyn_boot_strings_01}
    physicsBodyDefinitions [{
      sphereCollisionRadius=0
      boundBone {boneName=dyn_boot_strings_01}
      boxExtents=<10 10 10>
      constraintSetup {angularLimitsMax=<0 0 0>, angularLimitsMin=<0 0 0>, angularTarget=<0 0 0>, linearAxesMax=<4 2 0.02>, linearAxesMin=<-4 -2 -0.02>}
      localJointOffset=<0 0 -5>
     }]
   }
   n4 = Anim Dynamics - Bone: dyn_boot_strings_02(ComponentPose=@n3, Alpha=0.3, ExternalForce="0,0,0") {
    bGravityOverrideInSimSpace=true
    bLinearSpring=true
    bOverrideLinearDamping=true
    gravityScale=0.29
    linearDampingOverride=0.995
    linearSpringConstant=150
    simulationSpace=RootRelative
    alphaBoolBlend {blendOption=Linear}
    boundBone {boneName=dyn_boot_strings_02}
    physicsBodyDefinitions [{
      sphereCollisionRadius=0
      boundBone {boneName=dyn_boot_strings_02}
      boxExtents=<10 10 10>
      constraintSetup {angularLimitsMax=<0 0 0>, angularLimitsMin=<0 0 0>, angularTarget=<0 0 0>, linearAxesMax=<4 2 0.02>, linearAxesMin=<-5 -2 -0.02>}
      localJointOffset=<0 0 -6>
     }]
   }
   n5 = RigidBody(ComponentPose=@n4) {
    bClampLinearTranslationLimitToRefPose=true
    bOverrideWorldGravity=true
    overridePhysicsAsset=&/F_MED_Halloween_Bunny_Spring_Physics
    simulationSpace=BaseBoneSpace
    alphaBoolBlend {blendOption=Linear}
    baseBoneRef {boneName=root}
    componentAppliedLinearAccClamp=<1100 1100 1100>
    componentLinearAccScale=<0.5 0.5 0.5>
    componentLinearVelScale=<0.5 0.5 0.5>
    overrideWorldGravity {z=-990}
   }
   Output Pose(Result=Component To Local(ComponentPose=@n5)) {layerGroup=None}
   Get crouch_alpha() -> (crouch_alpha:double)
  ]}
 EventGraph {nodes [
   Event BlueprintUpdateAnimation()
   > n1 = Branch(Condition=Get bIsCrouching())
   > Set crouch_alpha(crouch_alpha:double=1) -> (Output_Get:double)
   Set crouch_alpha(execute=@n1.else, crouch_alpha=0)
   TryGetPawnOwner()
  ]}
]
variables [crouch_alpha {type=number}]
